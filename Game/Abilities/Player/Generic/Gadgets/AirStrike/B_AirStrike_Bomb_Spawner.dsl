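; Bomb spawner actor for the player's Air Strike gadget that drops a timed series of bomb projectiles.
; GA_AirStrike passes in the player pawn, spawn locations, explosion effect spec and origin transform through InitBombSpawn, which starts a looping timer.
; Each tick spawns one projectile at the next location, transformed by the origin and raised by a height offset, with zero initial speed and configurable gravity.
; After the last location it clears the timer and destroys itself.
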(event Custom|InitBombSpawn
  (params (InFortPlayerPawn :type Fort%20Player%20Pawn%20Object%20Reference) (InSpawnLocations :type Array%20of%20Vectors) (InEffectOnExplode :type Fort%20Gameplay%20Effect%20Container%20Spec%20Structure) (InBombInitTransform :type Transform))
  ;[1 1 1] Set values passed in by GA_AirStrike
  (Utilities|FlowControl|Sequence
    (:then_0
      (Variables|Default|SetFortPlayerPawn InFortPlayerPawn))
    (:then_1
      (Variables|Default|SetSpawnLocations InSpawnLocations))
    (:then_2
      (Variables|Default|SetEffectOnExplode InEffectOnExplode))
    (:then_3
      (Variables|Default|SetBombInitTransform InBombInitTransform))
    (:then_4
      ;[1 1 1] Init Spawn Timer Loop
      (bind _returnvalue (Utilities|Time|SetTimerbyEvent (delegate Custom|BombSpawnLoop) (Variables|Default|GetTimeBetweenBombs) true))
      (Variables|Default|SetSpawnBombsTimer _returnvalue))))

(event Custom|BombSpawnLoop
  (Projectile|SpawnProjectile :ProjectileClass (Variables|Default|GetBombPrj) :RequestedBy (Variables|Default|GetFortPlayerPawn) :SpawnLocation (+ (Math|Transform|TransformLocation (Variables|Default|GetBombInitTransform) (Utilities|Array|Get(acopy) (Variables|Default|GetSpawnLocations) (Variables|Default|GetSpawnLoopCounter))) (Math|Vector|MakeVector 0.0 0.0 (Variables|Default|GetHeightToSpawnBombs))) :SpawnRotation "0, 0, 0" :SpawnDirection (Variables|Default|GetBombSpawnDirection) :EffectContainerSpecToApplyOnExplode (Variables|Default|GetEffectOnExplode) :bUseDefaultPhysics false :InitialSpeed 0.0 :GravityScale (Variables|Default|GetBombPrjGravity) :ChargePercent 0.0)
  ;[1 1 1] Increment and check spawn count based on number of bomb spawns.
  (bind _output_get (Variables|Default|SetSpawnLoopCounter (+ (Variables|Default|GetSpawnLoopCounter) 1)))
  (if (not (< _output_get (Utilities|Array|Length (Variables|Default|GetSpawnLocations))))
    (Utilities|Time|ClearandInvalidateTimerbyHandle (Variables|Default|GetSpawnBombsTimer))
    (Actor|DestroyActor)))

(fn ConstructionScript)
