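; Blueprint actor for the air strike gadget's placement in the player abilities, named as a placement marker or preview.
; On BeginPlay it immediately destroys itself, and its construction script is empty.
; The graph holds no placement, targeting or damage logic of its own.
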(event EventBeginPlay
  (Actor|DestroyActor))

(fn ConstructionScript)
