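; Placement preview for the player's air strike gadget, showing particle reticules at the planned impact points.
; SwitchType receives a position pattern, clears the old emitters, plays a toggle cue and flags a rebuild.
; When placement becomes valid, visual updates spawn emitters with staggered delays, using one-hit or two-hit templates for duplicated points.
; While valid, one emitter's target refreshes per update and all interpolate toward theirs outside dedicated servers; toggle and destroy events hide them.
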(event EventDestroyed
  (bind _defaultpositonparticles (Variables|PlacementLoop|GetDefaultPositonParticles))
  (bind _current_particle_array (Variables|PlacementLoop|GetCurrent%20Particle%20Array))
  (Components|Activation|Deactivate _current_particle_array)
  (Rendering|SetVisibility _current_particle_array false true)
  (Development|SetHiddeninGame _current_particle_array true true)
  (Components|Activation|Deactivate _defaultpositonparticles)
  (Rendering|SetVisibility _defaultpositonparticles false true)
  (Development|SetHiddeninGame _defaultpositonparticles true true))

(event Custom|ToggleAirStrikeParticles
  (params (bInToggle :type Boolean))
  (bind _defaultpositonparticles (Variables|PlacementLoop|GetDefaultPositonParticles))
  (bind _current_particle_array (Variables|PlacementLoop|GetCurrent%20Particle%20Array))
  (Components|Activation|Deactivate _current_particle_array)
  (Rendering|SetVisibility _current_particle_array false true)
  (Development|SetHiddeninGame _current_particle_array true true)
  (Components|Activation|Deactivate _defaultpositonparticles)
  (Rendering|SetVisibility _defaultpositonparticles false true)
  (Development|SetHiddeninGame _defaultpositonparticles true true))

(event PlacementVisuals|EventOnBounceMotionUpdate
  (params (VectorScale :type Float%20%28single-precision%29)))

(event PlacementVisuals|EventOnInitializeDecoPreview
  (params (NewBuildingActor :type Building%20Actor%20Object%20Reference) (NewDecoItemDefinition :type Fort%20Deco%20Item%20Definition%20Object%20Reference)))

(event PlacementVisuals|EventOnUpdateScale
  (params (Scale :type Float%20%28single-precision%29) (InWorldSpacePivot :type Vector%20%28by%20ref%29 :default "0, 0, 0")))

(event PlacementVisuals|EventSetScalarParameterValueOnAllPreviewMIDs
  (params (ParamName :type Name :default "None") (ParamValue :type Float%20%28single-precision%29)))

(event Custom|RestartLineReticule)

(event PlacementVisuals|EventSetVectorParameterValueOnAllPreviewMIDs
  (params (ParamName :type Name :default "None") (ParamValue :type Linear%20Color%20Structure%20%28by%20ref%29)))

(event EventBeginPlay)

(event Custom|SwitchType
  (params (Positions%20With%20Duplicates :type Array%20of%20Vectors) (Cleaned%20Positions :type Array%20of%20Vectors) (Duplicated%20Entries :type Array%20of%20Integers))
  ;[0 0 0] Reinit patterns on right click and first time its brought up
  (Variables|PlacementLoop|SetPosition%20Array%20With%20duplicates Positions%20With%20Duplicates)
  (Variables|PlacementLoop|SetPosition%20Array%20Cleaned%20of%20duplicates Cleaned%20Positions)
  (Variables|PlacementLoop|SetDouble%20Entries Duplicated%20Entries)
  (Utilities|Array|Clear (Variables|PlacementLoop|GetTargetEmitterLocations))
  (Variables|PlacementLoop|SetMultiGateIndex 0)
  (CallFunction|KillEmittersAndStartTimes)
  (Variables|Default|SetRecreate%20particles%20if%20valid true)
  (Audio|PlaySound2D "/Game/Sounds/Fort_UI_Sounds/Gameplay/Airstrike_Placement_UI_Toggle_Cue.Airstrike_Placement_UI_Toggle_Cue"))

(event PlacementVisuals|EventOnUpdateVisuals
  (params (DeltaSeconds :type Float%20%28single-precision%29) (bSafeToPlace :type Boolean))
  (bind _multigateindex (Variables|PlacementLoop|GetMultiGateIndex))
  (bind _current_particle_array (Variables|PlacementLoop|GetCurrent%20Particle%20Array))
  (bind _output (Utilities|Array|Get(acopy) _current_particle_array _multigateindex))
  (bind _current_particle_array_2 (Variables|PlacementLoop|GetCurrent%20Particle%20Array))
  (bind _multigateindex_2 (Variables|PlacementLoop|GetMultiGateIndex))
  (if (Variables|Default|GetRecreate%20particles%20if%20valid)
    (Variables|Default|SetRecreate%20particles%20if%20valid false)
    (if bSafeToPlace
      _)
    (else
      ;[0 0 0] Valid designs?
      (bind _output_get (Variables|Default|SetPlacement%20is%20Valid%20Previously (Variables|Default|GetPlacement%20is%20Valid)))
      (bind _output_get_2 (Variables|Default|SetPlacement%20is%20Valid bSafeToPlace))
      ;;[0 0 0] Force Reinit on pattern change
      (bind _output_get_3 (Variables|Default|SetValid%20state%20changed (Math|Boolean|NotEqual(Boolean) _output_get_2 _output_get)))
      ;[0 0 0] Valid state changed
      (if _output_get_3
        ;[0 0 0] Placement is valid
        (if (Variables|Default|GetPlacement%20is%20Valid)
          ;[0 0 0] For loop that spawns particles
          ;;[1 1 1] init
          (Variables|PlacementLoop|SetForLoopCounter 0)
          (if (< (Variables|PlacementLoop|GetForLoopCounter) (Utilities|Array|Length (Variables|PlacementLoop|GetPosition%20Array%20Cleaned%20of%20duplicates)))
            ;;[0 0 0] Spawn particles at impact locations and assign correct template based on count
            (bind _impact_loc_world_space (Utilities|GetLocation (Utilities|Array|Get(acopy) (Variables|PlacementLoop|GetPosition%20Array%20Cleaned%20of%20duplicates) (Variables|PlacementLoop|GetForLoopCounter)) true)
              (:NewParam
                (bind _returnvalue (AddComponent|NODE_AddParticleSystemComponent-0))
                (Transformation|SetWorldLocation _returnvalue _impact_loc_world_space false false)
                (Effects|Components|ParticleSystem|SetTemplate _returnvalue (select (Utilities|Array|ContainsItem (Variables|PlacementLoop|GetDouble%20Entries) (Variables|PlacementLoop|GetForLoopCounter)) (Variables|PlacementVisuals|GetTwo%20Hit%20Particle%20Indicator) (Variables|PlacementVisuals|GetOne%20Hit%20Particle%20Indicator)))
                (Utilities|Array|Add (Variables|PlacementLoop|GetCurrent%20Particle%20Array) _returnvalue)
                ;;[0 0 0] For loop
                (Utilities|FlowControl|Delay (Variables|VFX|GetReticuleIntermittentDelay))
                (bind _output_get_4 (Variables|PlacementLoop|SetForLoopCounter (+ (stale _output_get_4) 1))))))
          ;[0 0 0] Completed
          (Utilities|FlowControl|Delay (Variables|Default|GetDelay%20Between%20Particle%20Spawns)) ; Rep counter
          (else
            _))
        (elif (Variables|Default|GetPlacement%20is%20Valid)
          ;[0 0 0] Sequentially update particle positions on tick
          (if (not (Networking|IsDedicatedServer))
            (Utilities|FlowControl|Sequence
              (:then_0
                (if (Utilities|IsValid _output :!class /Script/Engine.KismetSystemLibrary)
                  (bind _impact_loc_world_space_2 (Utilities|GetLocation (Transformation|GetWorldLocation _output))
                    (:NewParam
                      (Utilities|Array|SetArrayElem (Variables|PlacementLoop|GetTargetEmitterLocations) _multigateindex_2 _impact_loc_world_space_2 true)
                      (Utilities|Array|SetArrayElem (Variables|PlacementLoop|GetTargetEmitterNewMovementStartTimes) _multigateindex_2 (Utilities|Time|GetRealTimeSeconds) true))))
                (Variables|PlacementLoop|SetMultiGateIndex (Math|Integer|%25%28Integer%29 (+ (Variables|PlacementLoop|GetMultiGateIndex) 1) (+ (Utilities|Array|LastIndex _current_particle_array) 1))))
              (:then_1
                (for (_array_element _array_index) (Variables|PlacementLoop|GetTargetEmitterNewMovementStartTimes)
                  (if (Utilities|IsValid (Utilities|Array|Get(acopy) _current_particle_array_2 _array_index) :!class /Script/Engine.KismetSystemLibrary)
                    (Transformation|SetWorldLocation (Utilities|Array|Get(acopy) _current_particle_array_2 _array_index) (Math|Interpolation|VInterpTo (Transformation|GetWorldLocation (Utilities|Array|Get(acopy) _current_particle_array_2 _array_index)) (Utilities|Array|Get(acopy) (Variables|PlacementLoop|GetTargetEmitterLocations) _array_index) (- (Utilities|Time|GetRealTimeSeconds) _array_element) 0.25) false false))))))))
      ;[0 0 0] Kill Particles
      (CallFunction|KillEmittersAndStartTimes))))

(fn KillEmittersandStartTimes
  (bind _current_particle_array (Variables|PlacementLoop|GetCurrent%20Particle%20Array))
  (bind _defaultpositonparticles (Variables|PlacementLoop|GetDefaultPositonParticles))
  (Utilities|Array|Clear (Variables|PlacementLoop|GetTargetEmitterLocations))
  (Variables|PlacementLoop|SetMultiGateIndex 0)
  (Variables|PlacementLoop|SetForLoopCounter 0)
  (Utilities|Array|Clear (Variables|PlacementLoop|GetTargetEmitterNewMovementStartTimes))
  (Components|Activation|Deactivate _current_particle_array)
  (Utilities|Array|Clear _current_particle_array)
  (Components|Activation|Deactivate _defaultpositonparticles)
  (Utilities|Array|Clear _defaultpositonparticles)
  (Audio|Components|Audio|FadeOut (Variables|Default|Getairstrike_incoming_place_audiocomponent) 0.2 0.0 "Linear" :!class /Script/Engine.AudioComponent))

(fn ConstructionScript)
