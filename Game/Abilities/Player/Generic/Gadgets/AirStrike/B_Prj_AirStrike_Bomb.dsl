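; Air strike bomb projectile that marks its predicted ground impact and explodes on hit or stop.
; BeginPlay line-traces vertically to spawn a reticle, tinted friend or foe against the local player, with incoming sounds.
; An Exploded flag guards DoExplosion; OnExploded then plays particles, camera shake, sound, rumble and a decal reusing the BeginPlay trace normal, and removes the reticle.
; The damage event only runs a delay with nothing after it.
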(event EventBeginPlay
  (bind _returnvalue (Game|GetInstigatorController))
  (bind _returnvalue_2 (Game|GetPlayerController 0))
  (bind _returnvalue_3 (Team|OnSameTeam _returnvalue _returnvalue_2))
  (bind _returnvalue_4 (Transformation|GetActorTransform))
  ;[0.06995 0.1669 1] Set Team Colors
  (Variables|Default|SetTeam%20Color (Math|Color|SelectColor "(R=0.039847,G=0.104084,B=0.925000,A=1.000000)" "(R=0.770000,G=0.536567,B=0.053133,A=1.000000)" _returnvalue_3))
  (Variables|Default|SetFriendorFoe (Math|Conversions|ToFloat(Boolean) (not _returnvalue_3)))
  (Variables|Default|SetBombStartLocation (.location _returnvalue_4))
  (Variables|Default|SetBombStartTransform _returnvalue_4)
  ;[0.15 0.4777 1] Spawn Ground Impact Reticle
  (bind (_outhit _returnvalue_5) (Collision|LineTraceByChannel (Math|Vector|MakeVector (.x (Variables|Default|GetBombStartLocation)) (.y (Variables|Default|GetBombStartLocation)) (+ (.z (Variables|Default|GetBombStartLocation)) 15000)) (Math|Vector|MakeVector (.x (Variables|Default|GetBombStartLocation)) (.y (Variables|Default|GetBombStartLocation)) (- (.z (Variables|Default|GetBombStartLocation)) 15000)) "TraceTypeQuery1" true))
  (if _returnvalue_5
    (Networking|CanExecuteCosmeticEvents
      (:True
        (bind _returnvalue_6 (Effects|SpawnEmitteratLocation (Variables|Setup|GetP_ImpactReticle) (.Location (Collision|BreakHitResult _outhit))))
        (bind _output_get (Variables|Default|SetImpactReticle_Effect _returnvalue_6))
        (for _index (range (+ (Math|Integer|Clamp(Integer) (- (Rendering|Material|GetNumMaterials _output_get) 1) 0 99999) 1))
          (bind _returnvalue_7 (Rendering|Material|CreateDynamicMaterialInstance _output_get _index 0 "None" :!class /Script/Engine.PrimitiveComponent))
          (Rendering|Material|SetMaterial _output_get _index _returnvalue_7)
          (Rendering|Material|SetScalarParameterValue _returnvalue_7 "FriendOrFoe" (Variables|Default|GetFriendorFoe) :!class /Script/Engine.MaterialInstanceDynamic)
          (Audio|PlaySoundatLocation "/Game/Sounds/Fort_Gadgets/Airstrike/airstrike_incoming_Cue.airstrike_incoming_Cue" (.Location (Collision|BreakHitResult _outhit)))
          (Audio|PlaySoundatLocation "/Game/Sounds/Fort_Gadgets/Airstrike/airstrike_stereo_all_Cue.airstrike_stereo_all_Cue" (.Location (Collision|BreakHitResult _outhit))))))))

(event EventOnStoppedOnServer
  ;[0.7116 0.838 1] Explode On Impact
  (if (not (Variables|Default|GetExploded))
    (Variables|Default|SetExploded true)
    (Collision|SetCollisionEnabled (Variables|Collision|GetCapsuleComponent) "NoCollision")
    (Rendering|SetVisibility (Variables|BPrjGrenade|GetMesh) false false)
    (Delivery|DoExplosion)
    (Components|Activation|Deactivate (Variables|BPrjGrenade|GetTrail_Effect))))

(event Collision|EventHit
  (params (MyComp :type Primitive%20Component%20Object%20Reference) (Other :type Actor%20Object%20Reference) (OtherComp :type Primitive%20Component%20Object%20Reference) (bSelfMoved :type Boolean) (HitLocation :type Vector :default "0, 0, 0") (HitNormal :type Vector :default "0, 0, 0") (NormalImpulse :type Vector :default "0, 0, 0") (Hit :type Hit%20Result%20Structure%20%28by%20ref%29))
  ;[0.7116 0.838 1] Explode On Impact
  (if (not (Variables|Default|GetExploded))
    (Variables|Default|SetExploded true)
    (Rendering|SetVisibility (Variables|BPrjGrenade|GetMesh) false false)
    (CallFunction|StopProjectile Hit)
    (Delivery|DoExplosion)
    (Components|Activation|Deactivate (Variables|BPrjGrenade|GetTrail_Effect))))

(event Game|Damage|EventAnyDamage
  (params (Damage :type Float%20%28single-precision%29) (DamageType :type Damage%20Type%20Object%20Reference) (InstigatedBy :type Controller%20Object%20Reference) (DamageCauser :type Actor%20Object%20Reference))
  ;[0.7116 0.838 1] Explode On Impact
  (Utilities|FlowControl|Delay 0.1))

(event EventOnExploded
  (params (HitActors :type Array%20of%20Actor%20Object%20References) (HitResults :type Array%20of%20Hit%20Result%20Structures))
  (bind _impactreticle_effect (Variables|Default|GetImpactReticle_Effect))
  (Variables|Default|SetHitLocation (Transformation|GetActorLocation))
  (Networking|CanExecuteCosmeticEvents
    (:True
      (Rendering|SetVisibility (Variables|Default|GetGrid%20Static%20Mesh%20Decal) false false)
      (Utilities|FlowControl|Sequence
        (:then_0
          (bind _returnvalue (Effects|SpawnEmitteratLocation (Variables|VFX|GetP_Explosion) (Variables|Default|GetHitLocation)))
          (Variables|/Script/CascadeRuntime.ParticleSystemComponent|SetCustomTimeDilation (Variables|VFX|GetParticle%20Speed) _returnvalue)
          (Utilities|FlowControl|Sequence
            (:then_0
              (Camera|PlayWorldCameraShake (Variables|Default|GetCameraShake) (Variables|Default|GetHitLocation) 400 2500)
              (Utilities|EffectsQualityCheck 0
                (:Cinematic
                  _)
                (:Epic
                  _)
                (:High
                  _))
              (Utilities|SpawnDecalSimple "500.000000,500.000000,500.000000" "225.000000,225.000000,225.000000" (Variables|Default|GetHitLocation) (Math|Conversions|RotationFromXVector (.Normal (Collision|BreakHitResult (stale EventBeginPlay _outhit)))) 3.5 4.5 (Variables|VFX|GetDecalMaterial)))
            (:then_1
              (Audio|PlaySoundatLocation (Variables|Sounds|GetCue_CloseSound) (Variables|Default|GetHitLocation)) ; Play Close Explosion
            )))
        (:then_1
          ;[1 1 1] Kill_Impact Reticle
          (Utilities|IsValid _impactreticle_effect
            (:Is%20Valid
              (Components|Activation|Deactivate _impactreticle_effect))))
        (:then_2
          ;[1 1 1] Controller rumble, 1024 units around explosion.
          (Fortnite|PlayLocalForceFeedbackatLocation "/Game/Weapons/ForceFeedback/Explosions/FF_Generic_Explosion_Near.FF_Generic_Explosion_Near" (Transformation|GetActorLocation) 1024 "AirRaidExplosion"))))))

(fn StopProjectile
  (params (HitResult :type Hit%20Result%20Structure%20%28by%20ref%29))
  (Game|Components|ProjectileMovement|StopSimulating (Variables|Movement|GetProjectileMovementComponent) HitResult)
  (Collision|SetCollisionEnabled (Variables|Collision|GetCapsuleComponent) "NoCollision"))

(fn ConstructionScript
  (Transformation|AddLocalRotation (Variables|BPrjGrenade|GetMesh) (Math|Rotator|MakeRotator 0.0 0.0 (Math|Random|RandomFloatinRange 0.0 360)) false false))
