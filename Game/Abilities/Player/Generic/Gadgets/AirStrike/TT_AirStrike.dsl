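; Tooltip class for the Air Strike gadget ability, filling in ability-specific values in its description text.
; Initialization casts the ability instance to GA_AirStrike and runs its SetupAbility with the source ability system from the tooltip context.
; The MaxStacks token resolves to that ability's base bomb count as text.
; It reports the Airstrike Explode tag as its application tag.
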(fn GetApplicationTag
  (params (AbilityInstance :type Gameplay%20Ability%20Object%20Reference))
  (return "(TagName=\"Abilities.generic.Airstrike.Explode\")" true))

(fn GetTextForTokenFromAbilityInstance
  (params (AbilityInstance :type Gameplay%20Ability%20Object%20Reference) (Tag :type Gameplay%20Tag%20Structure%20%28by%20ref%29) (Context :type Fort%20Tooltip%20Context%20Object%20Reference) (Token :type Gameplay%20Tag%20Structure%20%28by%20ref%29))
  (Utilities|IsValid AbilityInstance
    (:Is%20Valid
      (bind _asga_air_strike (Utilities|Casting|CastToGA_AirStrike AbilityInstance)
        (:then
          (LocalVariables|SetAirStrike _asga_air_strike)
          (switch Utilities|FlowControl|Switch|SwitchonGameplayTag Token
            (:Default
              _)
            (:Tooltip.Token.Ability.MaxStacks
              (return (Utilities|Text|ToText(Integer) (Variables|/Game/Abilities/Player/Generic/Gadgets/AirStrike/GA_AirStrike.GA_AirStrike_C|GetBaseBombCount (LocalVariables|GetAirStrike))) true))))
        (:CastFailed
          _))
      (return "" false))
    (:Is%20Not%20Valid
      _)))

(fn InitializeAbilityInstance
  (params (AbilityInstance :type Gameplay%20Ability%20Object%20Reference) (Context :type Fort%20Tooltip%20Context%20Object%20Reference))
  (bind _asga_air_strike (Utilities|Casting|CastToGA_AirStrike AbilityInstance))
  (Class|GAAirStrike|SetupAbility _asga_air_strike (Variables|/Script/FortniteGame.FortTooltipContext|GetSourceAbilitySystem Context)))
